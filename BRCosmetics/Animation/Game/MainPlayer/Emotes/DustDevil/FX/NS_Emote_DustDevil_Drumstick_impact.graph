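# Niagara effect for drumstick impacts in the DustDevil emote, showing short-lived mesh and sprite bursts in a selectable color.
# User parameters, named as animation curves, set spawn rate, pick from a hard-coded six-color palette and tint particles.
# The sole active emitter merges two spawn groups with separate lifetime, offset, size and brightness through switches and a visibility tag.
# Two older single-group emitters are disabled, including the only HitIntensity-driven light.
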NiagaraSystem /BRCosmetics/Animation/Game/MainPlayer/Emotes/DustDevil/FX/NS_Emote_DustDevil_Drumstick_impact
dependencies {
 usedDataInterfaces [&NiagaraDataInterfaceCurve, &NiagaraDataInterfaceSocketReader]
 usedDynamicInputs [&FNTimeOfDayColor, &RandomRangeLinearColor]
 usedModules [
  &self:SystemSpawnScript.SystemScriptSource.SystemScriptGraph.NiagaraNodeAssignment_0.SetVariables_#1
  &self:SystemSpawnScript.SystemScriptSource.SystemScriptGraph.NiagaraNodeAssignment_1.SetVariables_#2
  &SystemState
 ]
}
emitters {
 Anim_curves_Color_Glow_test {
  bEnabled=false
  emitterUpdateScript [
   &EmitterState(Life Cycle Mode=::NewEnumerator1, Loop Behavior=::NewEnumerator1, Loop Duration=10, UseLoopDelay=false)
   &SpawnRate(SpawnRate=$User.Anim_Curve_spawnRate, Use Spawn Probability=false)
  ]
  particleSpawnScript [&InitializeParticle(
    Color=&SelectLinearColorFromArray(
     Array Sampling Mode=::NewEnumerator1
     Direct Array Index=&MakeIntFromFloat(Float=$User.Anim_curve_colorSelect)
     Direct Array Mode=::NewEnumerator1
     Color Selection Array=DataInterface() {propertyValues {ColorData [
        <rgba 0.0382 0.1329 0.7231 1>
        <rgba 0.7379 0.2582 0.0331 1>
        <rgba 0.0887 0.7231 0.1301 1>
        <rgba 1 0 0.7084 1>
        <rgba 0.5457 0.5271 0.063 1>
        <rgba 0.5457 0 0 1>
       ]}}
     Random Seed=?
    )
    Lifetime=0.13
    Position Offset=<0 0 -2>
    Sprite Size Mode=::NewEnumerator3
    Uniform Sprite Size=50
    UsePositionOffset=true
    Write Material Random=true
   )]
  particleUpdateScript [&ParticleState()]
  rendererClasses [&NiagaraSpriteRendererProperties]
  renderers [{rendererClass=&NiagaraSpriteRendererProperties}]
 }
 Anim_curves_spawn {
  bEnabled=false
  emitterUpdateScript [
   &EmitterState(Life Cycle Mode=::NewEnumerator1, Loop Behavior=::NewEnumerator1, Loop Duration=10, UseLoopDelay=false)
   &SpawnRate(SpawnRate=$User.Anim_Curve_spawnRate, Use Spawn Probability=false)
   &SpawnBurst_Instantaneous(Use Spawn Probability=false, UseLoopCountLimit=false) {enabled=false}
   &SpawnBurst_Instantaneous(Spawn Time=0.1, Use Spawn Probability=false, UseLoopCountLimit=false) {enabled=false, module=SpawnBurst_Instantaneous001}
   &SpawnBurst_Instantaneous(Spawn Time=0.2, Use Spawn Probability=false, UseLoopCountLimit=false) {enabled=false, module=SpawnBurst_Instantaneous002}
   &SpawnBurst_Instantaneous(Spawn Time=0.33, Use Spawn Probability=false, UseLoopCountLimit=false) {enabled=false, module=SpawnBurst_Instantaneous003}
   &SpawnBurst_Instantaneous(Spawn Time=0.43, Use Spawn Probability=false, UseLoopCountLimit=false) {enabled=false, module=SpawnBurst_Instantaneous004}
   &SpawnBurst_Instantaneous(Spawn Time=0.5, Use Spawn Probability=false, UseLoopCountLimit=false) {enabled=false, module=SpawnBurst_Instantaneous005}
  ]
  particleSpawnScript [
   &FN_Sprite_Attributes(Create Facing Attribute=true, Facing:Vector3f=<0 1 0>, Write to Module Namespace=true)
   &InitializeParticle(
    Color=&SelectLinearColorFromArray(
     Array Sampling Mode=::NewEnumerator1
     Direct Array Index=&MakeIntFromFloat(Float=$User.Anim_curve_colorSelect, Conversion Mode=::NewEnumerator1)
     Direct Array Mode=::NewEnumerator1
     Color Selection Array=DataInterface() {propertyValues {ColorData [
        <rgba 0.0382 0.1329 0.7231 1>
        <rgba 0.7379 0.2582 0.0331 1>
        <rgba 0.0887 0.7231 0.1301 1>
        <rgba 1 0 0.7084 1>
        <rgba 0.5457 0.5271 0.063 1>
        <rgba 0.5457 0 0 1>
       ]}}
     Random Seed=?
    )
    Lifetime=0.4
    Mesh Scale Mode=::NewEnumerator3
    Mesh Uniform Scale=0.5
    Position Offset=<0 5 0>
    Sprite Size Mode=::NewEnumerator3
    Uniform Sprite Size=&Multiply_Float(A=85, B=1)
    UsePositionOffset=true
    Write Material Random=true
   )
   &InitialMeshOrientation(Mesh Orientation Mode=::NewEnumerator2, Orientation Coordinate Space=::Local, Rotation=<0 0 360>, Use Rotation Vector=false) {enabled=false}
   &InitialMeshOrientation(
    Orientation Axis=<0 0 0>
    Orientation Vector=<0 0 0>
    Rotation=&RandomRangeVector(Minimum=<0 0 0>, Maximum=<1 0 0>)
    Use Orientation Vector=false
    Use Rotation Vector=true
   ) {enabled=false, module=InitialMeshOrientation001}
   /SetVariables(Particles.Intensity_Spawn:NiagaraFloat=&Add_Float(A=$User.HitIntensity, B=0)) {isSetParameters=true}
  ]
  particleUpdateScript [
   &ParticleState()
   &ScaleMeshSize(Scale Factor=&VectorFromFloat(Value=&FloatFromCurve(FloatCurve=DataInterface() {propertyValues {Curve {keys [{arriveTangent=0.5777, time=0, value=0.8}, {arriveTangent=0.4579, leaveTangent=0.4579, time=1, value=1}]}}})))
   &ScaleSpriteSize(Uniform Curve Scale=2.5)
   &ScaleColor(
    Scale Alpha=&FloatFromCurve(FloatCurve=DataInterface() {propertyValues {Curve {
        defaultValue=1
        keys [
         {arriveTangent=-3.8502, interpMode=RCIM_Cubic, leaveTangent=-3.8502, tangentMode=RCTM_User, time=0, value=1}
         {interpMode=RCIM_Cubic, tangentMode=RCTM_User, time=1, value=0}
        ]
       }}})
    Scale RGB=$User.Anim_curve_colorChange
    ScaleA=true
    ScaleRGB=true
   )
   &ScaleColor(
    Scale RGB=&VectorFromFloat(Value=&FNTimeOfDayScalar(Night Scale=5, Morning Scale=5, Day Scale=5, Evening Scale=5))
    ScaleA=true
    ScaleRGB=true
   ) {module=ScaleColor001}
   &DynamicMaterialParameters(
    Index 0 Param 1=&FloatFromCurve(FloatCurve=DataInterface() {propertyValues {Curve {keys [{arriveTangent=-2.1015, interpMode=RCIM_Cubic, leaveTangent=-2.1015, tangentMode=RCTM_User, time=0, value=0.5}, <at 1 -0.15>]}}})
    Index 0 Param 2=&FloatFromCurve(FloatCurve=DataInterface() {propertyValues {Curve {keys [{arriveTangent=-0.3464, interpMode=RCIM_Cubic, leaveTangent=-0.3464, tangentMode=RCTM_User, time=0, value=1}, <at 1 0.9>]}}})
    Index 0 Param 3=&FloatFromCurve(FloatCurve=DataInterface() {propertyValues {Curve {keys [{interpMode=RCIM_Cubic, tangentMode=RCTM_User, time=0, value=1}, {interpMode=RCIM_Cubic, tangentMode=RCTM_User, time=1, value=0.25}]}}})
    Module.Param0WriteEnabled=true
    Module.Param1WriteEnabled=true
    Module.Param2WriteEnabled=true
    Module.Param3WriteEnabled=false
   )
   &FN_Light_Attributes(
    Color=<rgba 0.2579 0.1262 1 1>
    Create Color Attribute=true
    Create Exponent Attribute=true
    Create Radius Attribute=true
    Exponent=5
    Radius=150
    Scale Color=&FloatFromCurve(
     FloatCurve=DataInterface() {propertyValues {Curve {keys [
         {arriveTangent=-1.7915, interpMode=RCIM_Cubic, leaveTangent=-1.7915, tangentMode=RCTM_User, time=0, value=1}
         {interpMode=RCIM_Cubic, tangentMode=RCTM_User, time=1, value=0}
        ]}}}
     Scale Curve=&Add_Float(A=3.5, B=&Multiply_Float(A=6, B=$Particles.Intensity_Spawn))
    )
   )
   /SetVariables(Particles.MeshOrientation=<0 1 1 0>) {isSetParameters=true, module=SetVariables_#3}
  ]
  rendererClasses [&NiagaraSpriteRendererProperties, &NiagaraMeshRendererProperties]
  renderers [
   {rendererClass=&NiagaraSpriteRendererProperties}
   {rendererClass=&NiagaraMeshRendererProperties, rendererIndex=1}
   {rendererClass=&NiagaraSpriteRendererProperties, rendererIndex=2}
   {rendererClass=&NiagaraMeshRendererProperties, rendererIndex=3}
  ]
 }
 Anim_curves_spawn001 {
  emitterUpdateScript [
   &EmitterState(Life Cycle Mode=::NewEnumerator1, Loop Behavior=::NewEnumerator1, Loop Duration=10, UseLoopDelay=false)
   &SpawnRate(SpawnGroup=1, SpawnRate=$User.Anim_Curve_spawnRate, Use Spawn Probability=false) {module=SpawnRate001}
   &SpawnRate(SpawnRate=$User.Anim_Curve_spawnRate, Use Spawn Probability=false)
  ]
  particleSpawnScript [
   /SetVariables(Particles.VisibilityTag=&SpawnGroup()) {isSetParameters=true}
   &FN_SetSpawnGroup(Number of Spawn Groups=2)
   &FN_Sprite_Attributes(Create Facing Attribute=true, Facing:Vector3f=<0 1 0>, Write to Module Namespace=true)
   &InitializeParticle(
    Color=&SelectLinearColorFromArray(
     Array Sampling Mode=::NewEnumerator1
     Direct Array Index=&MakeIntFromFloat(Float=$User.Anim_curve_colorSelect, Conversion Mode=::NewEnumerator1)
     Direct Array Mode=::NewEnumerator1
     Color Selection Array=DataInterface() {propertyValues {ColorData [
        <rgba 0.0382 0.1329 0.7231 1>
        <rgba 0.7379 0.2582 0.0331 1>
        <rgba 0.0887 0.7231 0.1301 1>
        <rgba 1 0 0.7084 1>
        <rgba 0.5457 0.5271 0.063 1>
        <rgba 0.5457 0 0 1>
       ]}}
     Random Seed=?
    )
    Lifetime=&MergeHelper_FloatSwitch(0=0.4, 1=0.13, 2=?, 3=?, Index=&SpawnGroup(), Num=::NewEnumerator0)
    Mesh Scale Mode=::NewEnumerator3
    Mesh Uniform Scale=0.5
    Position Offset=&MergeHelper_VectorSwitch(0=<0 5 0>, 1=<0 0 -2>, 2=?, 3=?, Index=&SpawnGroup(), Num=::NewEnumerator0)
    Sprite Size Mode=::NewEnumerator3
    Uniform Sprite Size=&MergeHelper_FloatSwitch(
     0=&Multiply_Float(A=85, B=1)
     1=88.1593
     2=?
     3=?
     Index=&SpawnGroup()
     Num=::NewEnumerator0
    )
    UsePositionOffset=true
    Write Material Random=true
   )
  ]
  particleUpdateScript [
   &ParticleState()
   &ScaleMeshSize(Scale Factor=&VectorFromFloat(Value=&FloatFromCurve(FloatCurve=DataInterface() {propertyValues {Curve {keys [{arriveTangent=0.5777, time=0, value=0.8}, {arriveTangent=0.4579, leaveTangent=0.4579, time=1, value=1}]}}})))
   &ScaleSpriteSize(
    Scale Sprite Size Mode=::NewEnumerator0
    Uniform Scale Factor=&MergeHelper_FloatSwitch(
     0=&FloatFromCurve(
      FloatCurve=DataInterface() {propertyValues {Curve {
         defaultValue=1
         keys [{interpMode=RCIM_Cubic, tangentMode=RCTM_User, time=0, value=0}, {interpMode=RCIM_Cubic, tangentMode=RCTM_User, time=1, value=1}]
        }}}
      Scale Curve=2.5
     )
     1=&FloatFromCurve(FloatCurve=DataInterface() {propertyValues {Curve {
         defaultValue=1
         keys [{interpMode=RCIM_Cubic, tangentMode=RCTM_User, time=0, value=0}, {interpMode=RCIM_Cubic, tangentMode=RCTM_User, time=0.3332, value=1}]
        }}})
     2=?
     3=?
     Index=&SpawnGroup()
     Num=::NewEnumerator0
    )
   ) {module=ScaleSpriteSize001}
   &ScaleColor(
    Scale Alpha=&FloatFromCurve(FloatCurve=DataInterface() {propertyValues {Curve {
        defaultValue=1
        keys [
         {arriveTangent=-3.8502, interpMode=RCIM_Cubic, leaveTangent=-3.8502, tangentMode=RCTM_User, time=0, value=1}
         {interpMode=RCIM_Cubic, tangentMode=RCTM_User, time=1, value=0}
        ]
       }}})
    Scale RGB=$User.Anim_curve_colorChange
    ScaleA=true
    ScaleRGB=true
   )
   &ScaleColor(
    Scale RGB=&VectorFromFloat(Value=&FNTimeOfDayScalar(
      Night Scale=&MergeHelper_FloatSwitch(0=5, 1=1, 2=?, 3=?, Index=&SpawnGroup(), Num=::NewEnumerator0)
      Morning Scale=&MergeHelper_FloatSwitch(0=5, 1=1, 2=?, 3=?, Index=&SpawnGroup(), Num=::NewEnumerator0)
      Day Scale=&MergeHelper_FloatSwitch(0=5, 1=1, 2=?, 3=?, Index=&SpawnGroup(), Num=::NewEnumerator0)
      Evening Scale=&MergeHelper_FloatSwitch(0=5, 1=1, 2=?, 3=?, Index=&SpawnGroup(), Num=::NewEnumerator0)
     ))
    ScaleA=true
    ScaleRGB=true
   ) {module=ScaleColor001}
   &DynamicMaterialParameters(
    Index 0 Param 1=&FloatFromCurve(FloatCurve=DataInterface() {propertyValues {Curve {keys [{arriveTangent=-2.1015, interpMode=RCIM_Cubic, leaveTangent=-2.1015, tangentMode=RCTM_User, time=0, value=0.5}, <at 1 -0.15>]}}})
    Index 0 Param 2=&FloatFromCurve(FloatCurve=DataInterface() {propertyValues {Curve {keys [{arriveTangent=-0.3464, interpMode=RCIM_Cubic, leaveTangent=-0.3464, tangentMode=RCTM_User, time=0, value=1}, <at 1 0.9>]}}})
    Index 0 Param 3=&FloatFromCurve(FloatCurve=DataInterface() {propertyValues {Curve {keys [{interpMode=RCIM_Cubic, tangentMode=RCTM_User, time=0, value=1}, {interpMode=RCIM_Cubic, tangentMode=RCTM_User, time=1, value=0.25}]}}})
    Module.Param0WriteEnabled=true
    Module.Param1WriteEnabled=true
    Module.Param2WriteEnabled=true
    Module.Param3WriteEnabled=false
   )
   /SetVariables(Particles.MeshOrientation=<0 1 1 0>) {isSetParameters=true, module=SetVariables_#3}
   /SetVariables(Particles.Glow Dynamic params:Vector4f=&MakeVector4(X=2, Y=0, Z=0, W=0)) {isSetParameters=true}
  ]
  rendererClasses [&NiagaraMeshRendererProperties, &NiagaraSpriteRendererProperties]
  renderers [
   {rendererClass=&NiagaraMeshRendererProperties, rendererIndex=0}
   {rendererClass=&NiagaraSpriteRendererProperties, rendererIndex=1}
   {rendererClass=&NiagaraSpriteRendererProperties, rendererIndex=2}
  ]
 }
}
summary {
 systemName=NS_Emote_DustDevil_Drumstick_impact
 emitters [
  {
   bEnabled=false
   emitterName=Anim_curves_spawn
   rendererClasses [&NiagaraSpriteRendererProperties, &NiagaraMeshRendererProperties]
  }
  {bEnabled=false, emitterName=Anim_curves_Color_Glow_test, rendererClasses [&NiagaraSpriteRendererProperties]}
  {
   emitterName=Anim_curves_spawn001
   rendererClasses [&NiagaraMeshRendererProperties, &NiagaraSpriteRendererProperties]
  }
 ]
 userVariables [
  User.Anim_curve_colorChange {type=Vector3f, description="", defaultValue {struct=Vector3f, value=<1 1 1>}}
  User.Anim_curve_colorSelect {type=NiagaraFloat, defaultValue=0, description=""}
  User.Anim_Curve_spawnRate {type=NiagaraFloat, defaultValue=2, description=""}
  User.Color Array {
   type=NiagaraDataInterfaceArrayColor
   description=""
   defaultValue {
    struct=NiagaraExt_VariableValue_DataInterface
    value {
     dataInterface=&self:NiagaraDataInterfaceArrayColor_0
     dataInterfaceClass=&NiagaraDataInterfaceArrayColor
    }
   }
  }
  User.Hit {type=NiagaraFloat, defaultValue=1, description=""}
  User.HitIntensity {type=NiagaraFloat, defaultValue=1, description=""}
  User.Hitted {type=NiagaraFloat, defaultValue=0, description=""}
  User.LeftRightHand {type=NiagaraFloat, defaultValue=0, description=""}
  User.RibbonIDOffset {type=NiagaraInt32}
 ]
}
