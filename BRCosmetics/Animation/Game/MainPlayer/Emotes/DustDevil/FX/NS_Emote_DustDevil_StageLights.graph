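# Niagara system for the DustDevil emote's stage lighting: a color-changing spotlight with rainbow bokeh and dust.
# SpotLightMain drives mesh, sprite, light and component renderers from a looping particle, picking its color by System.Bar and pulsing with System.Oscillation and Flicker.
# Bokeh and dust sprites draw from the same User.SpotlightColorArray, with bokeh alpha scaled by time of day.
# Beat-driven laser and spotlight-circle emitters are present but disabled.
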NiagaraSystem /BRCosmetics/Animation/Game/MainPlayer/Emotes/DustDevil/FX/NS_Emote_DustDevil_StageLights
dependencies {
 usedDataInterfaces [
  &NiagaraDataInterfaceCurve
  &NiagaraDataInterfaceArrayFloat
  &NiagaraDataInterfaceVectorCurve
  &NiagaraDataInterfaceColorCurve
 ]
 usedDynamicInputs [
  &MaskFloat
  &Divide_Float
  /Niagara/DynamicInputs/Divide/Divide_Int
  &ModuloInt
 ]
 usedModules [
  &SystemState
  &self:SystemSpawnScript.SystemScriptSource.SystemScriptGraph.NiagaraNodeAssignment_0.SetVariables_#1
  &self:SystemSpawnScript.SystemScriptSource.SystemScriptGraph.NiagaraNodeAssignment_1.SetVariables_#2
  &self:SystemSpawnScript.SystemScriptSource.SystemScriptGraph.NiagaraNodeAssignment_2.SetVariables_#3
 ]
}
emitters {
 Laser_idle_01001 {
  bEnabled=false
  emitterUpdateScript [
   &EmitterState(
    Delay First Loop Only=true
    Inactive Response=::NewEnumerator1
    Life Cycle Mode=::NewEnumerator1
    Loop Delay=0.27
    Loop Duration=&Multiply_Float(A=$System.CurrentLoopDuration, B=2)
    UseLoopDelay=true
   )
   &SpawnBurst_Instantaneous(
    Spawn Count=6
    Spawn Probability=&MakeCustomFloatFromBool(
     Bool=&SetBoolByFloatComparison(A=$User.LaserState, B=0, Comparison Type=::NewEnumerator4)
     True Float=1
     False Float=0
    )
    Use Spawn Probability=true
    UseLoopCountLimit=false
   )
  ]
  particleSpawnScript [
   &FN_SetSpawnGroup(Number of Spawn Groups=6)
   &InitializeParticle(
    Color Mode=::NewEnumerator0
    Lifetime Max=0.6
    Lifetime Min=0.8
    Lifetime Mode=::NewEnumerator1
    Mesh Scale=&Multiply_Vector(
     A=&SelectVectorFromArray(
      Direct Array Index=$Particles.SpawnGroup
      Vector Selection Array=DataInterface() {propertyValues {FloatData [<0.3 0.3 1.5>, <0.3 0.3 1.8>, <0.25 0.25 1.6>, <0.3 0.3 1.5>, <0.3 0.3 1.7>, <0.2 0.2 1.6>]}}
     )
     B=<1 1 0.8>
    )
    Mesh Scale Mode=::NewEnumerator4
    Position Offset=&Multiply_Vector(
     A=&SelectVectorFromArray(
      Direct Array Index=$Particles.SpawnGroup
      Vector Selection Array=DataInterface() {propertyValues {FloatData [<30 20 2>, <30 -20 2>, <0 -35 2>, <-40 -20 2>, <-40 20 2>, <0 35 2>]}}
     )
     B=<1 1 1>
    )
    Sprite Size Mode=::NewEnumerator3
    Uniform Sprite Size=77
    UsePositionOffset=true
    Write Material Random=true
   )
   &InitialMeshOrientation(
    Mesh Orientation Mode=::NewEnumerator7
    Rotation=&SelectVectorFromArray(
     Direct Array Index=$Particles.SpawnGroup
     Vector Selection Array=DataInterface() {propertyValues {FloatData [<0.2 0.5 -0.05>, <0.2 0.5 0.05>, <0 0.5 0.1>, <0 0.45 0.05>, <0 0.45 -0.05>, <0 0.5 -0.1>]}}
    )
    Use Rotation Vector=true
   )
   /SetVariables(Particles.GlowCameraOffset=33) {isSetParameters=true}
  ]
  particleUpdateScript [
   &ParticleState()
   &UpdateMeshOrientation(
    Rotation Coordinate Space=::NewEnumerator0
    Rotation Rate=&SelectFloatFromArray(Direct Array Index=$Particles.SpawnGroup, Float Selection Array=DataInterface() {propertyValues {FloatData [0.3,-0.4,0.35,-0.38,0.4,-0.35]}})
    Rotation Vector=<0 0 0.4>
   )
   &Color(Color=&SelectLinearColorFromArray(
     Array Sampling Mode=::NewEnumerator1
     Direct Array Index=&Add_Integer(A=$System.Beat, B=&RandomRangeInt(Minimum=0, Maximum=3))
     Direct Array Mode=::NewEnumerator1
     Color Selection Array=$User.SpotlightColorArray
     Random Seed=?
    )) {module=Color001}
   &ScaleColor(
    Color Value To Scale=$Particles.Color
    Scale Alpha=&MakeCustomFloatFromBool(
     Bool=&SetBoolByFloatComparison(A=$User.LaserState, B=0, Comparison Type=::NewEnumerator7)
     True Float=1
     False Float=0
    )
    ScaleA=true
    ScaleRGB=false
   ) {enabled=false, module=ScaleColor001}
   &ScaleColor(
    Color Value To Scale=$Particles.Color
    Scale Alpha=&FloatFromCurve(
     FloatCurve=DataInterface() {propertyValues {Curve {keys [
         {arriveTangent=5.6401, interpMode=RCIM_Cubic, leaveTangent=5.6401, tangentMode=RCTM_User, time=0, value=0}
         {arriveTangent=0.0567, interpMode=RCIM_Cubic, leaveTangent=0.0567, tangentMode=RCTM_User, time=0.2042, value=0.6228}
         {arriveTangent=0.0224, interpMode=RCIM_Cubic, leaveTangent=0.0224, tangentMode=RCTM_User, time=1, value=0}
        ]}}}
     Scale Curve=0.05
    )
    Scale RGB=&VectorFromFloat(Value=&FloatFromCurve(FloatCurve=DataInterface() {propertyValues {Curve {keys [<at 0 0>,<at 0.2 10>,<at 0.3 0.5>,<at 0.4 12>,<at 0.6 2>,<at 0.7 8.5>,<at 1 0>]}}}))
    ScaleA=true
    ScaleRGB=true
   )
   &ScaleMeshSize(Scale Factor=&MakeVector(X=1, Y=1, Z=1))
   /SetVariables(
    Particles.Color2=&MultiplyLinearColor(
     Linear Color=$Particles.Color
     Scale Factor=&MakeFloat4FromVectorAndFloat("Vector (RGB)"=&VectorFromFloat(Value=1000), "Float (Alpha)"=1)
    )
    Particles.Scale2:Vector3f=&Multiply_Vector(A=$Particles.Scale, B=<0.03 0.02 1>)
    Particles.ScaleThin:Vector3f=&Multiply_Vector(
     A=&Multiply_Vector(
      A=&SelectVectorFromArray(
       Direct Array Index=$Particles.PartitionParticles.Partition
       Direct Array Mode=::NewEnumerator1
       Vector Selection Array=DataInterface() {propertyValues {FloatData [<0.02 0.02 1.4>, <0.01 0.01 1.6>, <0.02 0.02 1.5>, <0.01 0.01 2>, <0.02 0.02 1.8>, <32.9576 35.5427 1.7>]}}
      )
      B=<4 4 1.2>
     )
     B=<0.03 0.02 1>
    )
    Particles.Size2:Vector2f=&Multiply_Vector2D(A=$Particles.SpriteSize, B=<2 2>)
   ) {isSetParameters=true}
  ]
  rendererClasses [&NiagaraSpriteRendererProperties, &NiagaraMeshRendererProperties]
  renderers [
   {rendererClass=&NiagaraSpriteRendererProperties}
   {rendererClass=&NiagaraMeshRendererProperties, rendererIndex=1}
  ]
 }
 Lasers_top {
  bEnabled=false
  emitterUpdateScript [
   &EmitterState(
    Inactive Response=::NewEnumerator1
    Life Cycle Mode=::NewEnumerator1
    Loop Behavior=::NewEnumerator1
    Loop Duration Mode=::NewEnumerator1
    UseLoopDelay=false
   )
   &SpawnBurst_Instantaneous(Spawn Count=6, Use Spawn Probability=false, UseLoopCountLimit=false)
   &SpawnRate(
    Spawn Probability=&MakeCustomFloatFromBool(
     Bool=&Boolean_AndOperation(
      Bool1=&SetBoolByIntComparison(A=$System.BarModulo, B=5, Comparison Type=::NewEnumerator4)
      Bool2=&Boolean_AndOperation(
       Bool1=&SetBoolByIntComparison(A=$System.BeatModulo, B=3, Comparison Type=::NewEnumerator2)
       Bool2=&SetBoolByIntComparison(A=$System.BeatModulo, B=2, Comparison Type=::NewEnumerator7)
      )
     )
     True Float=1
     False Float=0.5
    )
    SpawnRate=12
    Use Spawn Probability=true
   )
  ]
  particleSpawnScript [
   &FN_SetSpawnGroup(Number of Spawn Groups=6)
   &InitializeParticle(
    Color Mode=::NewEnumerator0
    Mesh Scale=&Multiply_Vector(
     A=&SelectVectorFromArray(
      Direct Array Index=$Particles.SpawnGroup
      Vector Selection Array=DataInterface() {propertyValues {FloatData [<0.3 0.3 1.5>, <0.3 0.3 1.8>, <0.25 0.25 1.6>, <0.3 0.3 1.5>, <0.3 0.3 1.7>, <0.2 0.2 1.6>]}}
     )
     B=<1 1 0.8>
    )
    Mesh Scale Mode=::NewEnumerator4
    Position Offset=&Multiply_Vector(
     A=&SelectVectorFromArray(
      Direct Array Index=$Particles.SpawnGroup
      Vector Selection Array=DataInterface() {propertyValues {FloatData [<30 20 -2>, <30 -20 -2>, <0 -35 -2>, <-40 -20 -2>, <-40 20 -2>, <0 35 -2>]}}
     )
     B=<1 1 3>
    )
    Sprite Size Mode=::NewEnumerator3
    Uniform Sprite Size=77
    UsePositionOffset=true
    Write Material Random=true
   )
   &InitialMeshOrientation(
    Mesh Orientation Mode=::NewEnumerator7
    Rotation=&SelectVectorFromArray(
     Direct Array Index=$Particles.SpawnGroup
     Vector Selection Array=DataInterface() {propertyValues {FloatData [<0.2 0.5 -0.05>, <0.2 0.5 0.05>, <0 0.5 0.1>, <0 0.45 0.05>, <0 0.45 -0.05>, <0 0.5 -0.1>]}}
    )
    Use Rotation Vector=true
   ) {enabled=false}
   /SetVariables(Particles.GlowCameraOffset=1) {isSetParameters=true, module=SetVariables_#4}
   &SystemLocation(
    MaskSpawnGroup=false
    Offset=&Multiply_Vector(
     A=&SelectVectorFromArray(
      Direct Array Index=$Particles.SpawnGroup
      Vector Selection Array=DataInterface() {propertyValues {FloatData [<30 20 2>, <30 -20 2>, <0 -35 2>, <-40 -20 2>, <-40 20 2>, <0 35 2>]}}
     )
     B=<1 1 1>
    )
   ) {enabled=false}
  ]
  particleUpdateScript [
   &ParticleState()
   &UpdateMeshOrientation(
    Facing Direction=/Niagara/DynamicInputs/Vectors/RotateVector(
     Vector To Rotate:Vector3f=&Lerp_Vector(
      A=<0 0 1>
      B=&NormalizeVector(
       Vector To Normalize=&Subtract_Position(
        A=$Particles.Position
        B=&AddVectorToPosition(Position=/Niagara/DynamicInputs/Helpers/SimulationPosition(), Vector=<0 0 1>)
       )
       Safe Normalize=false
       Safe Fallback Threshold=0
       Safe Fallback Vector=<0 0 0.001>
      )
      Alpha=&FloatFromCurve(
       FloatCurve=DataInterface() {propertyValues {Curve {keys [
           {interpMode=RCIM_Cubic, leaveTangentWeight=0.1889, time=0, value=1}
           {arriveTangent=3.3833, interpMode=RCIM_Cubic, leaveTangent=3.3833, time=0.216, value=5}
           <at 0.4433 2.5 RCIM_Cubic>
           {arriveTangentWeight=0.0703, interpMode=RCIM_Cubic, leaveTangentWeight=0.1394, tangentMode=RCTM_User, time=0.7568, value=5}
           {arriveTangentWeight=0.3366, interpMode=RCIM_Cubic, time=1, value=1.6166}
          ]}}}
       CurveIndex=&Frac_Float(Float=&Multiply_Float(A=$Engine.System.Age, B=0.2))
       Scale Curve=0.3
      )
     )
     Yaw:NiagaraFloat=&Lerp_Float(A=10, B=-10, Alpha=&Sine(Normalized Angle=$Emitter.Age, Period=10, Scale=0.5, Bias=0.5))
     Pitch:NiagaraFloat=5
     Roll:NiagaraFloat=0
     Rotation Rate:NiagaraFloat=1
    )
    Facing Interpolation=false
    Orientation Method=::NewEnumerator1
    Override Quat=false
    TransformFacing=false
    TransformUp=false
   ) {module=UpdateMeshOrientation001}
   &UpdateMeshOrientation(
    Rotation Coordinate Space=::NewEnumerator0
    Rotation Rate=&SelectFloatFromArray(Direct Array Index=$Particles.SpawnGroup, Float Selection Array=DataInterface() {propertyValues {FloatData [0.3,-0.4,0.35,-0.38,0.4,-0.35]}})
    Rotation Vector=<0 0 0.4>
   ) {enabled=false}
   &Color(
    Color=&SelectLinearColorFromArray(
     Array Sampling Mode=::NewEnumerator1
     Direct Array Index=$System.Bar
     Direct Array Mode=::NewEnumerator1
     Color Selection Array=$User.SpotlightColorArray
     Random Seed=?
    )
    Scale Alpha=&FloatFromCurve(FloatCurve=DataInterface() {propertyValues {Curve {
        defaultValue=0
        keys [
         {interpMode=RCIM_Cubic, tangentMode=RCTM_User, time=0, value=0.5}
         {interpMode=RCIM_Cubic, tangentMode=RCTM_User, time=0.5, value=1}
         {tangentMode=RCTM_User, time=1, value=0.5}
        ]
       }}})
    Scale Color=<5 5 5>
   ) {module=Color002}
   &Color(
    Color=&SelectLinearColorFromArray(
     Array Sampling Mode=::NewEnumerator1
     Direct Array Index=$System.Bar
     Direct Array Mode=::NewEnumerator1
     Color Selection Array=$User.SpotlightColorArray
     Random Seed=?
    )
    Scale Color=<0.25 0.25 0.25>
   ) {enabled=false}
   &ScaleMeshSize(Scale Factor=&VectorFromCurve(
     VectorCurve=DataInterface() {propertyValues {XCurve {keys [<at 0 0>,<at 0.2 0.8>]}, YCurve {keys [<at 0 0>,<at 0.2 0.8>]}, ZCurve {keys [<at 0 1>,<at 1 1>]}}}
     Scale Curve=<0.5 0.5 1>
    ))
   &ScaleSpriteSize(Uniform Curve Sprite Scale=DataInterface() {propertyValues {Curve {keys [{tangentMode=RCTM_User, time=0, value=0}, {tangentMode=RCTM_User, time=0.2, value=1}]}}})
   /SetVariables(Particles.GlowColor=&MultiplyLinearColor(Linear Color=$Particles.Color, Scale Factor=&MakeVector4_FromFloat(Float=0.5))) {isSetParameters=true, module=SetVariables_#5}
   &Color(
    Color=&ColorFromCurve(ColorCurve=DataInterface() {propertyValues {
       AlphaCurve {keys [<at 0 1>,<at 1 1>]}
       BlueCurve {keys [<at 0 0>,<at 0.333 0>,<at 0.667 1>,<at 1 0>]}
       GreenCurve {keys [<at 0 0>,<at 0.333 1>,<at 0.667 0>,<at 1 0>]}
       RedCurve {keys [<at 0 1>,<at 0.333 0>,<at 0.667 0>,<at 1 1>]}
      }})
    Scale Alpha=4.9322
   ) {enabled=false, module=Color001}
   &Color(
    Color=&SelectLinearColorFromArray(
     Array Sampling Mode=::NewEnumerator1
     Direct Array Index=$System.Bar
     Direct Array Mode=::NewEnumerator1
     Color Selection Array=$User.SpotlightColorArray
     Random Seed=?
    )
    Scale Alpha=&Multiply_Float(
     A=$System.Oscillation
     B=&SelectFloatFromArray(
      Direct Array Index=$System.LoopCount
      Direct Array Mode=::NewEnumerator1
      Float Selection Array=DataInterface() {propertyValues {FloatData [1,0.7,1,0.7,1,0.7]}}
     )
    )
   ) {enabled=false, module=Color003}
  ]
  rendererClasses [&NiagaraMeshRendererProperties, &NiagaraSpriteRendererProperties]
  renderers [
   {rendererClass=&NiagaraMeshRendererProperties, rendererIndex=0}
   {rendererClass=&NiagaraSpriteRendererProperties, rendererIndex=1}
  ]
 }
 Rainbow_bokeh {
  emitterUpdateScript [
   &EmitterState()
   &SpawnRate(SpawnRate=7, Use Spawn Probability=false)
  ]
  particleSpawnScript [
   &InitializeParticle(
    Color=&SelectLinearColorFromArray(
     Array Sampling Mode=::NewEnumerator0
     Direct Array Index=?
     Direct Array Mode=::NewEnumerator1
     Color Selection Array=$User.SpotlightColorArray
     Random Seed=0
    )
    Lifetime Max=1.8
    Lifetime Min=0.4
    Lifetime Mode=::NewEnumerator1
    Sprite Rotation Angle=&UniformAOrBFloat(A=0, B=0.125, Evaluation Type=::NewEnumerator0, Recalculate Random Each Loop=true, Override Seed=false)
    Sprite Rotation Mode=::NewEnumerator1
    Sprite Size Mode=::NewEnumerator1
    Uniform Sprite Size Max=20
    Uniform Sprite Size Min=10
    UsePositionOffset=true
    Write Material Random=true
   )
   &ShapeLocation(
    MaskSpawnGroup=false
    Non Uniform Scale=<0.7 0.7 0.8>
    Offset=<0 0 85>
    Offset Mode=::NewEnumerator0
    Override Seed=false
    Shape Origin=<0 0 0>
    UseOwnerScale=true
   )
   &CameraOffset(Camera Offset Amount=-100)
  ]
  particleUpdateScript [
   &ParticleState()
   &ScaleSpriteSize(Uniform Curve Sprite Scale=DataInterface() {propertyValues {Curve {
       defaultValue=0.6687
       keys [
        {interpMode=RCIM_Cubic, tangentMode=RCTM_User, time=0, value=0}
        {arriveTangent=-0.4264, interpMode=RCIM_Cubic, leaveTangent=-0.4264, tangentMode=RCTM_User, time=0.5, value=1}
       ]
      }}}) {enabled=false}
   &ScaleColor(
    Scale Alpha=&FloatFromCurve(
     FloatCurve=DataInterface() {propertyValues {Curve {keys [
         {arriveTangent=9.8157, interpMode=RCIM_Cubic, leaveTangent=9.8157, tangentMode=RCTM_User, time=0, value=0}
         {arriveTangent=-0.0381, interpMode=RCIM_Cubic, leaveTangent=-0.0381, tangentMode=RCTM_User, time=0.1675, value=0.6}
         {interpMode=RCIM_Cubic, tangentMode=RCTM_User, time=0.3, value=0.35}
         {interpMode=RCIM_Cubic, tangentMode=RCTM_User, time=0.5, value=0.5}
         {interpMode=RCIM_Cubic, tangentMode=RCTM_Break, time=1, value=0}
        ]}}}
     Scale Curve=&FNTimeOfDayScalar(Night Scale=0.2, Morning Scale=0.35, Day Scale=0.4, Evening Scale=0.35)
    )
    ScaleA=true
    ScaleRGB=true
   )
   &Drag(Drag=2.5, Use Linear Drag=true, Use Rotational Drag=false) {enabled=false}
   &GravityForce(Gravity=&RandomRangeVector(Minimum=<-5 -5 -5>, Maximum=<5 -5 5>))
   &SolveForcesAndVelocity(Clamp Velocity=false, Limit Acceleration=false)
   /SetVariables(Particles.New Position:NiagaraPosition=&Add_Vector(
     A=Linked() {linkedVariable {name=Particles.Position, type=NiagaraPosition}}
     B=Linked() {linkedVariable {name=Engine.Owner.Position, type=NiagaraPosition}}
    )) {isSetParameters=true}
  ]
  rendererClasses [&NiagaraSpriteRendererProperties]
  renderers [{rendererClass=&NiagaraSpriteRendererProperties}]
 }
 Rainbow_dust {
  emitterUpdateScript [
   &EmitterState(
    Inactive Response=::NewEnumerator1
    Life Cycle Mode=::NewEnumerator1
    Loop Behavior=::NewEnumerator1
    Loop Duration Mode=::NewEnumerator1
    UseLoopDelay=false
   )
   &SpawnRate(SpawnRate=5, Use Spawn Probability=true)
  ]
  particleSpawnScript [
   &InitializeParticle(
    Color=&SelectLinearColorFromArray(
     Array Sampling Mode=::NewEnumerator0
     Direct Array Index=?
     Direct Array Mode=::NewEnumerator1
     Color Selection Array=$User.SpotlightColorArray
     Random Seed=0
    )
    Lifetime=&RandomRangeFloat(Minimum=1, Maximum=1.5)
    Sprite Rotation Mode=::NewEnumerator2
    Sprite Size Mode=::NewEnumerator1
    Uniform Sprite Size Max=35
    Uniform Sprite Size Min=25
    UsePositionOffset=false
    Write Material Random=true
   )
   &AddVelocity(Velocity=<0 0 0>)
   &SystemLocation(MaskSpawnGroup=false)
   &ShapeLocation(MaskSpawnGroup=false, Override Seed=false, Sphere Radius=75, UseOwnerScale=true)
  ]
  particleUpdateScript [
   &ParticleState()
   &SubUVAnimation(
    End Frame=15
    SubUV Animation Mode:ENiagara_SubUVLookupMode=::NewEnumerator1
    SubUV Curve=DataInterface() {propertyValues {Curve {keys [{tangentMode=RCTM_User, time=0, value=0}, {tangentMode=RCTM_User, time=1.0102, value=0.6518}]}}}
   )
   &Drag(
    Drag=&Multiply_Float(
     A=&RandomRangeFloat(Minimum=0.5, Maximum=1)
     B=&FloatFromCurve(FloatCurve=DataInterface() {propertyValues {Curve {keys [
          {interpMode=RCIM_Cubic, leaveTangent=0.248, tangentMode=RCTM_Break, time=0, value=0}
          {interpMode=RCIM_Cubic, tangentMode=RCTM_Break, time=1, value=1}
         ]}}})
    )
    Use Linear Drag=true
    Use Rotational Drag=false
   )
   &SolveForcesAndVelocity(Clamp Velocity=false, Limit Acceleration=false)
   &ScaleColor(
    Scale Alpha=&FloatFromCurve(
     FloatCurve=DataInterface() {propertyValues {Curve {keys [
         {arriveTangent=5, interpMode=RCIM_Cubic, time=0, value=0}
         {arriveTangent=1.5181, interpMode=RCIM_Cubic, leaveTangent=1.5181, time=0.2, value=0.7}
         {
          arriveTangent=0.1817
          arriveTangentWeight=0.1842
          interpMode=RCIM_Cubic
          leaveTangent=0.1817
          leaveTangentWeight=0.7369
          tangentMode=RCTM_User
          time=0.5037
          value=0.7647
         }
         {arriveTangent=-1.5409, interpMode=RCIM_Cubic, leaveTangent=-1.5409, time=0.8, value=0.7376}
         <at 1 0>
        ]}}}
     Scale Curve=&RandomRangeFloat(Minimum=0.02, Maximum=0.07)
    )
    ScaleA=true
    ScaleRGB=true
   )
   &ScaleSpriteSize(
    Uniform Curve Scale=10
    Uniform Curve Sprite Scale=DataInterface() {propertyValues {Curve {keys [
        {interpMode=RCIM_Cubic, leaveTangent=0.6647, tangentMode=RCTM_Break, time=0, value=1}
        {arriveTangent=-0.0319, interpMode=RCIM_Cubic, tangentMode=RCTM_Break, time=1.0059, value=1.1871}
       ]}}}
   )
  ]
  rendererClasses [&NiagaraSpriteRendererProperties]
  renderers [{rendererClass=&NiagaraSpriteRendererProperties}]
 }
 SpotLightMain {
  emitterUpdateScript [
   &EmitterState(
    Inactive Response=::NewEnumerator1
    Life Cycle Mode=::NewEnumerator1
    Loop Behavior=::NewEnumerator1
    Loop Duration Mode=::NewEnumerator1
    UseLoopDelay=false
   )
   &SpawnBurst_Instantaneous(Use Spawn Probability=false, UseLoopCountLimit=false)
  ]
  particleSpawnScript [
   &InitializeParticle(
    Mesh Scale=<0.8 0.8 1.2>
    Mesh Scale Mode=::NewEnumerator4
    Sprite Size Mode=::NewEnumerator3
    Uniform Sprite Size=333
    UsePositionOffset=false
    Write Material Random=true
   )
   &CameraOffset()
   /SetVariables(
    Particles.FloorGlowFacingVector:Vector3f=<0.01 0.01 1>
    Particles.FloorGlowPosition:NiagaraPosition=&AddVectorToPosition(Position=&SimulationPosition(), Vector=<0 0 1>)
    Particles.FloorGlowSize:Vector2f=<500 500>
    Particles.LightAttenuationScale=275
    Particles.LightExponent=2
    Particles.LightPosition:NiagaraPosition=&AddVectorToPosition(Position=&SimulationPosition(), Vector=<0 0 200>)
    Particles.LightRadius=230
    Particles.SpotLightPosition:NiagaraPosition=&AddVectorToPosition(Position=&SimulationPosition(), Vector=<0 0 280>)
   ) {isSetParameters=true}
  ]
  particleUpdateScript [
   &ParticleState(
    Kill Particles When Lifetime Has Elapsed=false
    Let Infinitely Lived Particles Die When Emitter Deactivates=true
    Loop Particles Lifetime=true
   )
   &Color(
    Color=&SelectLinearColorFromArray(
     Array Sampling Mode=::NewEnumerator1
     Direct Array Index=$System.Bar
     Direct Array Mode=::NewEnumerator1
     Color Selection Array=$User.SpotlightColorArray
     Random Seed=?
    )
    Scale Alpha=&Multiply_Float(
     A=$System.Oscillation
     B=&SelectFloatFromArray(
      Direct Array Index=$System.LoopCount
      Direct Array Mode=::NewEnumerator1
      Float Selection Array=DataInterface() {propertyValues {FloatData [1,0.7,1,0.7,1,0.7]}}
     )
    )
   )
   /SetVariables(
    Particles.CoreSpotlightColor:LinearColor=&MultiplyLinearColor(
     Linear Color=<rgba 0.2 0.1687 0.16 0.5>
     Scale Factor=&MakeFloat4FromVectorAndFloat("Vector (RGB)"=<1 1 1>, "Float (Alpha)"=$System.Flicker)
    )
    Particles.FlareColor=&MultiplyLinearColor(Linear Color=$Particles.Color, Scale Factor=&MakeVector4_FromFloat(Float=0.4))
    Particles.FloorGlowColor:LinearColor=&MultiplyLinearColor(Linear Color=$Particles.Color, Scale Factor=&MakeVector4_FromFloat(Float=0.3))
    Particles.LightColor=&MultiplyLinearColor(Linear Color=$Particles.Color, Scale Factor=&MakeVector4_FromFloat(Float=8))
    Particles.LightComponentIntensity:NiagaraFloat=&Multiply_Float(A=1, B=500)
   ) {isSetParameters=true}
  ]
  rendererClasses [
   &NiagaraMeshRendererProperties
   &NiagaraLightRendererProperties
   &NiagaraComponentRendererProperties
   &NiagaraSpriteRendererProperties
  ]
  renderers [
   {rendererClass=&NiagaraMeshRendererProperties, rendererIndex=0}
   {rendererClass=&NiagaraMeshRendererProperties, rendererIndex=1}
   {rendererClass=&NiagaraLightRendererProperties, rendererIndex=2}
   {rendererClass=&NiagaraComponentRendererProperties, rendererIndex=3}
   {rendererClass=&NiagaraSpriteRendererProperties, rendererIndex=4}
  ]
 }
 SpotlightCircle {
  bEnabled=false
  emitterUpdateScript [
   &EmitterState()
   &SpawnBurst_Instantaneous(
    Spawn Count=6
    Spawn Probability=&MakeCustomFloatFromBool(
     Bool=&SetBoolByFloatComparison(A=$User.LaserState, B=0, Comparison Type=::NewEnumerator4)
     True Float=1
     False Float=0
    )
    Use Spawn Probability=true
    UseLoopCountLimit=false
   ) {enabled=false}
   &SpawnRate(
    Spawn Probability=&MakeCustomFloatFromBool(
     Bool=&Boolean_AndOperation(
      Bool1=&SetBoolByIntComparison(A=$System.BarModulo, B=5, Comparison Type=::NewEnumerator4)
      Bool2=&Boolean_AndOperation(
       Bool1=&SetBoolByIntComparison(A=$System.BeatModulo, B=3, Comparison Type=::NewEnumerator2)
       Bool2=&SetBoolByIntComparison(A=$System.BeatModulo, B=2, Comparison Type=::NewEnumerator7)
      )
     )
     True Float=1
     False Float=0.5
    )
    SpawnRate=12
    Use Spawn Probability=true
   )
  ]
  particleSpawnScript [
   &FN_SetSpawnGroup(Number of Spawn Groups=6)
   &InitializeParticle(
    Color Mode=::NewEnumerator0
    Lifetime Max=0.6
    Lifetime Min=0.8
    Lifetime Mode=::NewEnumerator1
    Mesh Scale=&Multiply_Vector(
     A=&SelectVectorFromArray(
      Direct Array Index=$Particles.SpawnGroup
      Vector Selection Array=DataInterface() {propertyValues {FloatData [<0.3 0.3 1.5>, <0.3 0.3 1.8>, <0.25 0.25 1.6>, <0.3 0.3 1.5>, <0.3 0.3 1.7>, <0.2 0.2 1.6>]}}
     )
     B=<1 1 0.8>
    )
    Mesh Scale Mode=::NewEnumerator4
    Position Offset=&Multiply_Vector(
     A=&SelectVectorFromArray(
      Direct Array Index=$Particles.SpawnGroup
      Vector Selection Array=DataInterface() {propertyValues {FloatData [<30 20 -2>, <30 -20 -2>, <0 -35 -2>, <-40 -20 -2>, <-40 20 -2>, <0 35 -2>]}}
     )
     B=<1 1 3>
    )
    Sprite Size Mode=::NewEnumerator3
    Uniform Sprite Size=77
    UsePositionOffset=true
    Write Material Random=true
   )
   &InitialMeshOrientation(
    Mesh Orientation Mode=::NewEnumerator7
    Rotation=&SelectVectorFromArray(
     Direct Array Index=$Particles.SpawnGroup
     Vector Selection Array=DataInterface() {propertyValues {FloatData [<0.2 0.5 -0.05>, <0.2 0.5 0.05>, <0 0.5 0.1>, <0 0.45 0.05>, <0 0.45 -0.05>, <0 0.5 -0.1>]}}
    )
    Use Rotation Vector=true
   )
   /SetVariables(Particles.GlowCameraOffset=1) {isSetParameters=true, module=SetVariables_#4}
   &SystemLocation(
    MaskSpawnGroup=false
    Offset=&Multiply_Vector(
     A=&SelectVectorFromArray(
      Direct Array Index=$Particles.SpawnGroup
      Vector Selection Array=DataInterface() {propertyValues {FloatData [<30 20 2>, <30 -20 2>, <0 -35 2>, <-40 -20 2>, <-40 20 2>, <0 35 2>]}}
     )
     B=<1 1 1>
    )
   ) {enabled=false}
  ]
  particleUpdateScript [
   &ParticleState()
   &UpdateMeshOrientation(
    Facing Direction=&RotateVector(
     Vector To Rotate:Vector3f=&Lerp_Vector(
      A=<0 0 1>
      B=&NormalizeVector(
       Vector To Normalize=&Subtract_Position(
        A=$Particles.Position
        B=&AddVectorToPosition(Position=&SimulationPosition(), Vector=<0 0 1>)
       )
       Safe Normalize=false
       Safe Fallback Threshold=0
       Safe Fallback Vector=<0 0 0.001>
      )
      Alpha=&FloatFromCurve(
       FloatCurve=DataInterface() {propertyValues {Curve {keys [
           {interpMode=RCIM_Cubic, leaveTangentWeight=0.1889, time=0, value=1}
           {arriveTangent=3.3833, interpMode=RCIM_Cubic, leaveTangent=3.3833, time=0.216, value=5}
           <at 0.4433 2.5 RCIM_Cubic>
           {arriveTangentWeight=0.0703, interpMode=RCIM_Cubic, leaveTangentWeight=0.1394, tangentMode=RCTM_User, time=0.7568, value=5}
           {arriveTangentWeight=0.3366, interpMode=RCIM_Cubic, time=1, value=1.6166}
          ]}}}
       CurveIndex=&Frac_Float(Float=&Multiply_Float(A=$Engine.System.Age, B=0.2))
       Scale Curve=0.3
      )
     )
     Yaw:NiagaraFloat=&Lerp_Float(A=10, B=-10, Alpha=&Sine(Normalized Angle=$Emitter.Age, Period=10, Scale=0.5, Bias=0.5))
     Pitch:NiagaraFloat=5
     Roll:NiagaraFloat=0
     Rotation Rate:NiagaraFloat=1
    )
    Facing Interpolation=false
    Orientation Method=::NewEnumerator1
    Override Quat=false
    TransformFacing=false
    TransformUp=false
   ) {module=UpdateMeshOrientation001}
   &UpdateMeshOrientation(
    Rotation Coordinate Space=::NewEnumerator0
    Rotation Rate=&SelectFloatFromArray(Direct Array Index=$Particles.SpawnGroup, Float Selection Array=DataInterface() {propertyValues {FloatData [0.3,-0.4,0.35,-0.38,0.4,-0.35]}})
    Rotation Vector=<0 0 0.4>
   ) {enabled=false}
   &Color(
    Color=&SelectLinearColorFromArray(
     Array Sampling Mode=::NewEnumerator1
     Direct Array Index=$System.Bar
     Direct Array Mode=::NewEnumerator1
     Color Selection Array=$User.SpotlightColorArray
     Random Seed=?
    )
    Scale Alpha=&FloatFromCurve(FloatCurve=DataInterface() {propertyValues {Curve {
        defaultValue=0
        keys [
         {interpMode=RCIM_Cubic, tangentMode=RCTM_User, time=0, value=0}
         {interpMode=RCIM_Cubic, tangentMode=RCTM_User, time=0.5, value=1}
         {tangentMode=RCTM_User, time=1, value=0}
        ]
       }}})
    Scale Color=<5 5 5>
   ) {module=Color002}
   &Color(
    Color=&SelectLinearColorFromArray(
     Array Sampling Mode=::NewEnumerator1
     Direct Array Index=$System.Bar
     Direct Array Mode=::NewEnumerator1
     Color Selection Array=$User.SpotlightColorArray
     Random Seed=?
    )
    Scale Color=<0.25 0.25 0.25>
   ) {enabled=false}
   &ScaleMeshSize(Scale Factor=&VectorFromCurve(
     VectorCurve=DataInterface() {propertyValues {XCurve {keys [<at 0 0>,<at 0.2 0.8>]}, YCurve {keys [<at 0 0>,<at 0.2 0.8>]}, ZCurve {keys [<at 0 1>,<at 1 1>]}}}
     Scale Curve=<0.5 0.5 1>
    ))
   &ScaleSpriteSize(Uniform Curve Sprite Scale=DataInterface() {propertyValues {Curve {keys [{tangentMode=RCTM_User, time=0, value=0}, {tangentMode=RCTM_User, time=0.2, value=1}]}}})
   /SetVariables(Particles.GlowColor=&MultiplyLinearColor(Linear Color=$Particles.Color, Scale Factor=&MakeVector4_FromFloat(Float=0.5))) {isSetParameters=true, module=SetVariables_#5}
   &Color(
    Color=&ColorFromCurve(ColorCurve=DataInterface() {propertyValues {
       AlphaCurve {keys [<at 0 1>,<at 1 1>]}
       BlueCurve {keys [<at 0 0>,<at 0.333 0>,<at 0.667 1>,<at 1 0>]}
       GreenCurve {keys [<at 0 0>,<at 0.333 1>,<at 0.667 0>,<at 1 0>]}
       RedCurve {keys [<at 0 1>,<at 0.333 0>,<at 0.667 0>,<at 1 1>]}
      }})
    Scale Alpha=4.9322
   ) {enabled=false, module=Color001}
  ]
  rendererClasses [&NiagaraMeshRendererProperties, &NiagaraSpriteRendererProperties]
  renderers [
   {rendererClass=&NiagaraMeshRendererProperties, rendererIndex=0}
   {rendererClass=&NiagaraSpriteRendererProperties, rendererIndex=1}
  ]
 }
 SpotlightCircle002 {
  bEnabled=false
  emitterUpdateScript [
   &EmitterState()
   &SpawnBurst_Instantaneous(Spawn Count=6, Use Spawn Probability=true, UseLoopCountLimit=false)
   &SpawnRate(
    Spawn Probability=&MakeCustomFloatFromBool(
     Bool=&Boolean_AndOperation(
      Bool1=&SetBoolByIntComparison(A=$System.BarModulo, B=5, Comparison Type=::NewEnumerator4)
      Bool2=&Boolean_AndOperation(
       Bool1=&SetBoolByIntComparison(A=$System.BeatModulo, B=3, Comparison Type=::NewEnumerator2)
       Bool2=&SetBoolByIntComparison(A=$System.BeatModulo, B=2, Comparison Type=::NewEnumerator7)
      )
     )
     True Float=1
     False Float=0.5
    )
    SpawnRate=12
    Use Spawn Probability=true
   ) {enabled=false}
  ]
  particleSpawnScript [
   &FN_SetSpawnGroup(Number of Spawn Groups=6)
   &InitializeParticle(
    Color Mode=::NewEnumerator0
    Lifetime Max=0.6
    Lifetime Min=0.8
    Lifetime Mode=::NewEnumerator1
    Mesh Scale=&Multiply_Vector(
     A=&SelectVectorFromArray(
      Direct Array Index=$Particles.SpawnGroup
      Vector Selection Array=DataInterface() {propertyValues {FloatData [<0.3 0.3 1.5>, <0.3 0.3 1.8>, <0.25 0.25 1.6>, <0.3 0.3 1.5>, <0.3 0.3 1.7>, <0.2 0.2 1.6>]}}
     )
     B=<1 1 0.8>
    )
    Mesh Scale Mode=::NewEnumerator4
    Position Offset=&Multiply_Vector(
     A=&SelectVectorFromArray(
      Direct Array Index=$Particles.SpawnGroup
      Vector Selection Array=DataInterface() {propertyValues {FloatData [<30 20 -2>, <30 -20 -2>, <0 -35 -2>, <-40 -20 -2>, <-40 20 -2>, <0 35 -2>]}}
     )
     B=<1 1 3>
    )
    Sprite Size Mode=::NewEnumerator3
    Uniform Sprite Size=77
    UsePositionOffset=true
    Write Material Random=true
   )
   &InitialMeshOrientation(
    Mesh Orientation Mode=::NewEnumerator7
    Rotation=&SelectVectorFromArray(
     Direct Array Index=$Particles.SpawnGroup
     Vector Selection Array=DataInterface() {propertyValues {FloatData [<0.2 0.5 -0.05>, <0.2 0.5 0.05>, <0 0.5 0.1>, <0 0.45 0.05>, <0 0.45 -0.05>, <0 0.5 -0.1>]}}
    )
    Use Rotation Vector=true
   ) {enabled=false}
   /SetVariables(Particles.GlowCameraOffset=1) {isSetParameters=true, module=SetVariables_#4}
   &SystemLocation(
    MaskSpawnGroup=false
    Offset=&Multiply_Vector(
     A=&SelectVectorFromArray(
      Direct Array Index=$Particles.SpawnGroup
      Vector Selection Array=DataInterface() {propertyValues {FloatData [<30 20 2>, <30 -20 2>, <0 -35 2>, <-40 -20 2>, <-40 20 2>, <0 35 2>]}}
     )
     B=<1 1 1>
    )
   ) {enabled=false}
  ]
  particleUpdateScript [
   &ParticleState()
   &UpdateMeshOrientation(
    Facing Direction=&RotateVector(
     Vector To Rotate:Vector3f=&Lerp_Vector(
      A=<0 0 1>
      B=&NormalizeVector(
       Vector To Normalize=&Subtract_Position(
        A=$Particles.Position
        B=&AddVectorToPosition(Position=&SimulationPosition(), Vector=<0 0 1>)
       )
       Safe Normalize=false
       Safe Fallback Threshold=0
       Safe Fallback Vector=<0 0 0.001>
      )
      Alpha=&FloatFromCurve(
       FloatCurve=DataInterface() {propertyValues {Curve {keys [
           {interpMode=RCIM_Cubic, leaveTangentWeight=0.1889, time=0, value=1}
           {arriveTangent=3.3833, interpMode=RCIM_Cubic, leaveTangent=3.3833, time=0.216, value=5}
           <at 0.4433 2.5 RCIM_Cubic>
           {arriveTangentWeight=0.0703, interpMode=RCIM_Cubic, leaveTangentWeight=0.1394, tangentMode=RCTM_User, time=0.7568, value=5}
           {arriveTangentWeight=0.3366, interpMode=RCIM_Cubic, time=1, value=1.6166}
          ]}}}
       CurveIndex=&Frac_Float(Float=&Multiply_Float(A=$Engine.System.Age, B=0.2))
       Scale Curve=0.3
      )
     )
     Yaw:NiagaraFloat=&Lerp_Float(A=10, B=-10, Alpha=&Sine(Normalized Angle=$Emitter.Age, Period=10, Scale=0.5, Bias=0.5))
     Pitch:NiagaraFloat=5
     Roll:NiagaraFloat=0
     Rotation Rate:NiagaraFloat=1
    )
    Facing Interpolation=false
    Orientation Method=::NewEnumerator1
    Override Quat=false
    TransformFacing=false
    TransformUp=false
   ) {module=UpdateMeshOrientation001}
   &UpdateMeshOrientation(
    Rotation Coordinate Space=::NewEnumerator0
    Rotation Rate=&SelectFloatFromArray(Direct Array Index=$Particles.SpawnGroup, Float Selection Array=DataInterface() {propertyValues {FloatData [0.3,-0.4,0.35,-0.38,0.4,-0.35]}})
    Rotation Vector=<0 0 0.4>
   ) {enabled=false}
   &Color(
    Color=&SelectLinearColorFromArray(
     Array Sampling Mode=::NewEnumerator1
     Direct Array Index=$System.Bar
     Direct Array Mode=::NewEnumerator1
     Color Selection Array=$User.SpotlightColorArray
     Random Seed=?
    )
    Scale Alpha=&FloatFromCurve(FloatCurve=DataInterface() {propertyValues {Curve {
        defaultValue=0
        keys [
         {interpMode=RCIM_Cubic, tangentMode=RCTM_User, time=0, value=0}
         {interpMode=RCIM_Cubic, tangentMode=RCTM_User, time=0.5, value=1}
         {tangentMode=RCTM_User, time=1, value=0}
        ]
       }}})
    Scale Color=<5 5 5>
   ) {module=Color002}
   &Color(
    Color=&SelectLinearColorFromArray(
     Array Sampling Mode=::NewEnumerator1
     Direct Array Index=$System.Bar
     Direct Array Mode=::NewEnumerator1
     Color Selection Array=$User.SpotlightColorArray
     Random Seed=?
    )
    Scale Color=<0.25 0.25 0.25>
   ) {enabled=false}
   &ScaleMeshSize(Scale Factor=&VectorFromCurve(
     VectorCurve=DataInterface() {propertyValues {XCurve {keys [<at 0 0>,<at 0.2 0.8>]}, YCurve {keys [<at 0 0>,<at 0.2 0.8>]}, ZCurve {keys [<at 0 1>,<at 1 1>]}}}
     Scale Curve=<0.5 0.5 1>
    ))
   &ScaleSpriteSize(Uniform Curve Sprite Scale=DataInterface() {propertyValues {Curve {keys [{tangentMode=RCTM_User, time=0, value=0}, {tangentMode=RCTM_User, time=0.2, value=1}]}}})
   /SetVariables(Particles.GlowColor=&MultiplyLinearColor(Linear Color=$Particles.Color, Scale Factor=&MakeVector4_FromFloat(Float=0.5))) {isSetParameters=true, module=SetVariables_#5}
   &Color(
    Color=&ColorFromCurve(ColorCurve=DataInterface() {propertyValues {
       AlphaCurve {keys [<at 0 1>,<at 1 1>]}
       BlueCurve {keys [<at 0 0>,<at 0.333 0>,<at 0.667 1>,<at 1 0>]}
       GreenCurve {keys [<at 0 0>,<at 0.333 1>,<at 0.667 0>,<at 1 0>]}
       RedCurve {keys [<at 0 1>,<at 0.333 0>,<at 0.667 0>,<at 1 1>]}
      }})
    Scale Alpha=4.9322
   ) {enabled=false, module=Color001}
  ]
  rendererClasses [&NiagaraMeshRendererProperties, &NiagaraSpriteRendererProperties]
  renderers [
   {rendererClass=&NiagaraMeshRendererProperties, rendererIndex=0}
   {rendererClass=&NiagaraSpriteRendererProperties, rendererIndex=1}
  ]
 }
}
summary {
 systemName=NS_Emote_DustDevil_StageLights
 emitters [
  {
   bEnabled=false
   emitterName=SpotlightCircle
   rendererClasses [&NiagaraMeshRendererProperties, &NiagaraSpriteRendererProperties]
  }
  {
   emitterName=SpotLightMain
   rendererClasses [
    &NiagaraMeshRendererProperties
    &NiagaraLightRendererProperties
    &NiagaraComponentRendererProperties
    &NiagaraSpriteRendererProperties
   ]
  }
  {
   bEnabled=false
   emitterName=Laser_idle_01001
   rendererClasses [&NiagaraSpriteRendererProperties, &NiagaraMeshRendererProperties]
  }
  {emitterName=Rainbow_bokeh, rendererClasses [&NiagaraSpriteRendererProperties]}
  {emitterName=Rainbow_dust, rendererClasses [&NiagaraSpriteRendererProperties]}
  {
   bEnabled=false
   emitterName=Lasers_top
   rendererClasses [&NiagaraMeshRendererProperties, &NiagaraSpriteRendererProperties]
  }
  {
   bEnabled=false
   emitterName=SpotlightCircle002
   rendererClasses [&NiagaraMeshRendererProperties, &NiagaraSpriteRendererProperties]
  }
 ]
 userVariables [
  User.BeamColorScale {type=NiagaraFloat, defaultValue=0.05, description=""}
  User.bIsFrontEnd {type=NiagaraBool}
  User.bIsFrontEndPreview {type=NiagaraBool}
  User.Color {type=LinearColor, description="", defaultValue {struct=LinearColor, value=<rgba 0.4817 0.5 0.4734 1>}}
  User.Color LUT Custom U Position {type=NiagaraFloat, defaultValue=0, description=""}
  User.FeetRestingPositions {
   type=NiagaraDataInterfaceArrayFloat3
   description=""
   defaultValue {
    struct=NiagaraExt_VariableValue_DataInterface
    value {
     dataInterface=&self:NiagaraDataInterfaceArrayFloat3_2
     dataInterfaceClass=&NiagaraDataInterfaceArrayFloat3
    }
   }
  }
  User.FlashesColorArray {
   type=NiagaraDataInterfaceArrayColor
   description=""
   defaultValue {
    struct=NiagaraExt_VariableValue_DataInterface
    value {
     dataInterface=&self:NiagaraDataInterfaceArrayColor_2
     dataInterfaceClass=&NiagaraDataInterfaceArrayColor
    }
   }
  }
  User.Jam? {type=NiagaraBool, defaultValue=true}
  User.LaserLengthMult {type=NiagaraFloat, defaultValue=1, description=""}
  User.LasersPositionArray {
   type=NiagaraDataInterfaceArrayFloat3
   description=""
   defaultValue {
    struct=NiagaraExt_VariableValue_DataInterface
    value {
     dataInterface=&self:NiagaraDataInterfaceArrayFloat3_3
     dataInterfaceClass=&NiagaraDataInterfaceArrayFloat3
    }
   }
  }
  User.LaserState {type=NiagaraFloat, defaultValue=0, description=""}
  User.Local Alpha Preview {type=NiagaraFloat, defaultValue=0.75, description=""}
  User.Other Alpha Preview {type=NiagaraFloat, defaultValue=0.25, description=""}
  User.SongBPM {type=NiagaraFloat, defaultValue=135, description=""}
  User.SpotlightColorArray {
   type=NiagaraDataInterfaceArrayColor
   description=""
   defaultValue {
    struct=NiagaraExt_VariableValue_DataInterface
    value {
     dataInterface=&self:NiagaraDataInterfaceArrayColor_0
     dataInterfaceClass=&NiagaraDataInterfaceArrayColor
    }
   }
  }
  User.SpotlightsOrientationArray {
   type=NiagaraDataInterfaceArrayFloat3
   description=""
   defaultValue {
    struct=NiagaraExt_VariableValue_DataInterface
    value {
     dataInterface=&self:NiagaraDataInterfaceArrayFloat3_1
     dataInterfaceClass=&NiagaraDataInterfaceArrayFloat3
    }
   }
  }
  User.SpotlightsPositionArray {
   type=NiagaraDataInterfaceArrayFloat3
   description=""
   defaultValue {
    struct=NiagaraExt_VariableValue_DataInterface
    value {
     dataInterface=&self:NiagaraDataInterfaceArrayFloat3_0
     dataInterfaceClass=&NiagaraDataInterfaceArrayFloat3
    }
   }
  }
  User.UseCustomColor {type=NiagaraFloat, defaultValue=1, description=""}
  User.UseMaterialColorOverride {type=NiagaraFloat, defaultValue=0, description=""}
 ]
}
